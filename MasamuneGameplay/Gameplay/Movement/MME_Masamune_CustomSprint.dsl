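; Movement mode extension that customizes tactical sprint for the Masamune weapon, marking players who jump out of a sprint.
; On dedicated-server activation, it binds the pawn's jump event and, if Masamune is equipped, records the current weapon as activating item, otherwise logging.
; Jumping applies GE_Masamune_InAirFromSprint; deactivation unbinds the event.
; An OnMovementModeChanged function, not bound here, removes the InAirFromSprint tag when the previous mode is not Falling.
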(event EventOnExtensionActivate
  (params (RuntimeContext :type MMERuntime%20Context%20Structure%20%28by%20ref%29))
  (bind _ownerpawn (Variables|FortMovementMode_BaseExtLogic|GetOwnerPawn))
  ;[0.7899 0.1002 0.2833] Apply a tag to the player when they tac sprint jump
  (if (CallFunction|IsServer?)
    (Pawn|BindEventtoOnPawnJumped (Variables|FortMovementMode_BaseExtLogic|GetOwnerPawn) (delegate Custom|OnJumped))
    (bind _returnvalue (Class|FortMovementModeExtBlueprintLibrary|GetRuntimeDatafromContext RuntimeContext "/Script/FortniteGame.FortMovementMode_ItemActivatedRuntimeData"))
    (if (GameplayTags|HasMatchingGameplayTag _ownerpawn (Variables|Default|GetMasamuneEquippedTag))
      (Class|FortMovementModeItemActivatedRuntimeData|SetActivatingItemfromWeapon _returnvalue (Variables|/Script/FortniteGame.FortPawn|GetCurrentWeapon _ownerpawn))
      (else
        (Fortnite|FortShippingLog _ownerpawn "MME_Masamune_CustomSprint activated but the Masamune Equipped tag is missing from the OwnerPawn! Cannot set activating item from weapon reliably." true)))))

(event Custom|OnJumped
  ;[0.7899 0.1002 0.2833] Apply a tag to the player when they tac sprint jump
  (GameplayEffects|ApplyGameplayEffectToSelf :self (Ability|GetFortAbilitySystemComponent (Variables|FortMovementMode_BaseExtLogic|GetOwnerPawn)) :GameplayEffectClass "/MasamuneGameplay/Gameplay/Effects/GE_Masamune_InAirFromSprint.GE_Masamune_InAirFromSprint_C" :Level 0.0))

(event EventOnExtensionDeactivate
  (params (RuntimeContext :type MMERuntime%20Context%20Structure%20%28by%20ref%29))
  ;[0.7899 0.1002 0.2833] Unbind things
  (if (CallFunction|IsServer?)
    (Pawn|UnbindEventfromOnPawnJumped (Variables|FortMovementMode_BaseExtLogic|GetOwnerPawn) (EventDispatchers|CreateEvent :!event Custom|OnJumped))))

(fn IsServer?
  (return (Networking|IsDedicatedServer (Variables|FortMovementMode_BaseExtLogic|GetOwnerPawn))))

(fn OnMovementModeChanged
  (params (Character :type Character%20Object%20Reference) (PrevMovementMode :type EMovementMode%20Enum :default "MOVE_None") (PreviousCustomMode :type Byte))
  (if (Utilities|Enum|NotEqual(Enum) PrevMovementMode "MOVE_Falling")
    (CallFunction|RemoveJumpedTag)))

(fn RemoveJumpedTag
  (bind _ownerpawn (Variables|FortMovementMode_BaseExtLogic|GetOwnerPawn))
  (if (GameplayTags|HasMatchingGameplayTag _ownerpawn "(TagName=\"Granted.Masamune.InAirFromSprint\")")
    (Ability|GameplayEffect|RemoveLooseGameplayTags _ownerpawn (GameplayTags|MakeGameplayTagContainerfromTag "(TagName=\"Granted.Masamune.InAirFromSprint\")") true)))
